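# Sound cue for the Shadow Bomb item's expire tell, by its name a warning that the effect is about to end.
# A single wave player feeds a Modulator that applies a slight random upward pitch shift at constant volume for variation.
# It plays in the gadget items sound class, using a shared GhostRocks attenuation asset for spatial falloff.
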SoundCue /FNGameplayCues/Athena/Sounds/Items/Shadow_Bomb/Sahdow_Bomb_Expire_Tell_Cue
root=Modulator_0
nodes [
 {id=Modulator_0, children [WavePlayer_1], properties {pitchMax=1.1, pitchMin=1.05, volumeMax=1, volumeMin=1}}
 {id=WavePlayer_1, properties {soundWaveAssetPtr=&/WPN_Shadow_Bomb_Expire_Tell_01}}
]
properties {
 attenuationSettings=/Game/Sounds/Attenuation/Fort_GhostRocks_Att
 maxDistance=4500
 soundClassObject=&Fort_Gadgets_Items
 volumeMultiplier=1
}
