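# Sound cue named for the activation of the Shadow Bomb item, set up in the gadget items sound class.
# Its only node is a root Modulator with an empty child input, so the cue references no sound wave and has nothing to play as shown.
# It uses the shared GhostRocks attenuation asset and a reduced volume multiplier.
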SoundCue /FNGameplayCues/Athena/Sounds/Items/Shadow_Bomb/Shadow_Bomb_Activate_Cue
root=Modulator_0
nodes [{id=Modulator_0, children [null]}]
properties {
 attenuationSettings=/Game/Sounds/Attenuation/Fort_GhostRocks_Att
 maxDistance=4500
 soundClassObject=&Fort_Gadgets_Items
 volumeMultiplier=0.6
}
