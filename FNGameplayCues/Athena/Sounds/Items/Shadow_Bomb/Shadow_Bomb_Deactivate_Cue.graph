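# Sound cue for the Shadow Bomb deactivation, playing the exit-stealth sound in Athena gameplay.
# A wave player for a single exit-stealth wave feeds a modulator node at the root, which is left at its default settings.
# It routes through the gadget items sound class and borrows the GhostRocks attenuation asset for spatial falloff.
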SoundCue /FNGameplayCues/Athena/Sounds/Items/Shadow_Bomb/Shadow_Bomb_Deactivate_Cue
root=Modulator_1
nodes [
 {id=Modulator_1, children [WavePlayer_1]}
 {id=WavePlayer_1, properties {soundWaveAssetPtr=&/WPN_Shadow_Bomb_Exit_Stealth_01}}
]
properties {
 attenuationSettings=/Game/Sounds/Attenuation/Fort_GhostRocks_Att
 maxDistance=4500
 soundClassObject=&Fort_Gadgets_Items
}
