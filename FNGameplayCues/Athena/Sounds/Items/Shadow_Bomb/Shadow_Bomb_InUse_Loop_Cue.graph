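# Looping sound cue for the in-use (stealth) state of the Shadow Bomb item in Athena.
# A continuous modulator scales pitch and volume of a looping stealth wave from a Velocity parameter, so the loop varies with speed.
# It plays as a gadget item sound with Ghost Rocks attenuation and overrides concurrency to stop the quietest instance beyond a cap.
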SoundCue /FNGameplayCues/Athena/Sounds/Items/Shadow_Bomb/Shadow_Bomb_InUse_Loop_Cue
root=ModulatorContinuous_0
nodes [
 {
  id=ModulatorContinuous_0
  children [WavePlayer_5]
  properties {
   pitchModulationParams {default=0, maxOutput=1.5, minOutput=0.8, parameterName=Velocity}
   volumeModulationParams {default=0, maxOutput=1.5, minOutput=0.75, parameterName=Velocity}
  }
 }
 {
  id=WavePlayer_5
  properties {bLooping=true, soundWaveAssetPtr=&/WPN_Shadow_Bomb_Stealth_Loop_01}
 }
]
properties {
 attenuationSettings=/Game/Sounds/Attenuation/Fort_GhostRocks_Att
 bOverrideConcurrency=true
 maxDistance=4500
 soundClassObject=&Fort_Gadgets_Items
 volumeMultiplier=1.5
 concurrencyOverrides {maxCount=8, resolutionRule=StopQuietest}
}
