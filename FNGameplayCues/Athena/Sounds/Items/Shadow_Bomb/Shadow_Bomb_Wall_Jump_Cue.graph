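# Sound cue for the wall jump sound of the Shadow Bomb item.
# A Random node picks one of four equally weighted wave variations to play.
# It is routed to the gadget items sound class at reduced volume and uses the GhostRocks attenuation asset.
# A concurrency override caps simultaneous instances and stops the quietest when the limit is reached.
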SoundCue /FNGameplayCues/Athena/Sounds/Items/Shadow_Bomb/Shadow_Bomb_Wall_Jump_Cue
root=Random_0
nodes [
 {id=Random_0, children [WavePlayer_4,WavePlayer_5,WavePlayer_6,WavePlayer_7], properties {weights [1,1,1,1]}}
 {id=WavePlayer_4, properties {soundWaveAssetPtr=&/WPN_Shadow_Bomb_Wall_Jump_01}}
 {id=WavePlayer_5, properties {soundWaveAssetPtr=&/WPN_Shadow_Bomb_Wall_Jump_02}}
 {id=WavePlayer_6, properties {soundWaveAssetPtr=&/WPN_Shadow_Bomb_Wall_Jump_03}}
 {id=WavePlayer_7, properties {soundWaveAssetPtr=&/WPN_Shadow_Bomb_Wall_Jump_04}}
]
properties {
 attenuationSettings=/Game/Sounds/Attenuation/Fort_GhostRocks_Att
 bOverrideConcurrency=true
 maxDistance=4500
 soundClassObject=&Fort_Gadgets_Items
 volumeMultiplier=0.65
 concurrencyOverrides {maxCount=8, resolutionRule=StopQuietest}
}
